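# Opaque, Nanite-ready base material for Hermes instanced ground foliage, with static switches selecting forest-floor leaf, gravel, scree or simple textured looks.
# Leaves blend per-biome palettes by vertex color and biome masks, while gravel and scree use biome-color functions and detail noise.
# Wetness and snow stages adjust the surface.
# An optional runtime virtual texture path blends into landscape by height, bypassed on forward and mobile shading paths.
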Material /BRCoreMaterials/Hermes/Foliage/Opaque/BaseMaterial/M_Hermes_Foliage_Opaque
graph {
 properties {
  bTangentSpaceNormal=false
  bUseMaterialAttributes=true
  bUsedWithInstancedStaticMeshes=true
  bUsedWithNanite=true
  pixelDepthOffsetMode=PDOM_Legacy
  usedShadingModels=Default Lit
 }
 nodes [
  n1 = Reroute()
  n2 = GetMaterialAttributes(None=@n1) {attributeGetTypes [#1, #2, #3]}
  n3 = ScalarParameter() {defaultValue=1, parameterName=Wet_AlbedoDarken}
  n4 = ScalarParameter() {defaultValue=0.5, parameterName=Wet_RoughnessMult}
  BlendMaterialAttributes(
   A=SetMaterialAttributes!(
    MaterialAttributes=@n2.MaterialAttributes
    Base Color=Multiply(A=@n2.Base Color, B=@n3)
    Specular=@n2.MaterialAttributes
    Roughness=Multiply(A=@n2.Roughness, B=@n4)
   ) {attributeSetTypes [#1, #2, #3]}
   B=@n1
   Alpha=NamedRerouteUsage()
  )
  Constant3Vector() {constant=<rgba 0.2 0.151 0.1423 0>}
  Constant3Vector() {constant=<rgba 0.1 0.056 0.02 0>}
  Constant3Vector() {constant=<rgba 0.07 0.0499 0.028 1>}
  Constant3Vector() {constant=<rgba 0.08 0.0607 0.04 1>}
  MaterialFunctionCall!(Scale=ScalarParameter() {defaultValue=64, group=Scree, parameterName=Scree_DetailNoiseScale}) {materialFunction=/Game/Athena/Apollo/Environments/Landscape/Materials/Biome/CommonMaterials/MF_SimpleWorldUV}
  n5 = NamedRerouteUsage()
  n6 = NamedRerouteUsage()
  n7 = StaticSwitchParameter(
   True=NamedRerouteUsage()
   False=StaticSwitchParameter(True=NamedRerouteUsage(), False=StaticSwitchParameter(True=@n5, False=@n6) {group=Scree, parameterName=IsScree, sortPriority=-1000}) {group=Gravel, parameterName=IsGravel, sortPriority=-1000}
  ) {group=ForestFloorLeaf_Settings, parameterName=IsForestFloorLeaves, sortPriority=-1000}
  MaterialFunctionCall!(
   In=@n7
   FlipTwoSidedSign=StaticBoolParameter() {group=Snow, parameterName=SnowFlipTwosidedNormal}
   IsTangentSpaceInput=StaticBool()
   HasBackFlipFaceNormals=StaticBool() {value=true}
  ) {materialFunction=/Game/Packages/Fortnite_UniqueMaterials_1/MaterialFunctions/Trees/MF_Tree_Snow}
  n8 = TextureSampleParameter2D() {
   group=Textures
   parameterName=Texture_BaseColor
   samplerSource=SSM_Wrap_WorldGroupSettings
   texture=/Game/Global/Textures/Default/Blanks/T_MidgreysRGB
  }
  n9 = TextureSampleParameter2D() {
   group=Textures
   parameterName=Texture_Specular
   samplerSource=SSM_Wrap_WorldGroupSettings
   samplerType=SAMPLERTYPE_Masks
   texture=/Game/Global/Textures/Default/Blanks/Temp_specular_Mask
  }
  n10 = TextureSampleParameter2D() {
   group=Textures
   parameterName=Texture_Normal
   samplerSource=SSM_Wrap_WorldGroupSettings
   samplerType=SAMPLERTYPE_Normal
   texture=&FlatNormalBlue
  }
  n11 = SetMaterialAttributes!(
   MaterialAttributes
   Base Color=NamedRerouteDeclaration(None=@n8.RGB) {name=Texture_BaseColor, nodeColor=<rgba 0 1 0.1562 1>}
   Specular=NamedRerouteDeclaration(None=@n9.R) {name=Texture_Specular, nodeColor=<rgba 0 1 0.1562 1>}
   Roughness=NamedRerouteDeclaration(None=@n9.B) {name=Texture_Roughness, nodeColor=<rgba 0 1 0.1562 1>}
   Normal=NamedRerouteDeclaration(None=@n10.RGB) {name=Texture_Normal, nodeColor=<rgba 0 1 0.1562 1>}
   Opacity Mask=NamedRerouteDeclaration(None=StaticSwitchParameter(True=@n9.G, False=@n8.A) {defaultValue=true, desc="False=UseDiffuseAlpha", group=OpacityMask, parameterName=UseGChannelOfSpecularForOpacityMask}) {name=Texture_OpacityMask, nodeColor=<rgba 0 1 0.1562 1>}
  ) {attributeSetTypes [
    #1
    #2
    #3
    #4
    #5
   ]}
  NamedRerouteDeclaration(None=@n11) {name=Simple_Material, nodeColor=<rgba 1 0 0.2109 1>}
  n12 = Add(
   A=Multiply(A=TextureCoordinate(), B=ScalarParameter() {defaultValue=2.5, group=Scree, parameterName=Scree_DetailNoiseScaleUV0})
   B=PerInstanceRandom()
  )
  n13 = TextureSample(UVs=@n12) {samplerSource=SSM_Wrap_WorldGroupSettings, samplerType=SAMPLERTYPE_Masks, texture=/Game/Athena/WinterArt/Textures/T_DirtLeaks_M}
  n14 = MaterialFunctionCall(Float3=@n13.RGB) {materialFunction=&BreakOutFloat3Components}
  NamedRerouteDeclaration(None=@n14.B) {name=Details-B, nodeColor=<rgba 0.4062 1 0 1>}
  n15 = TextureCoordinate()
  n16 = ScalarParameter() {defaultValue=1.5, group=Gravel, parameterName=Gravel_DetailNoiseScaleUV0}
  n17 = PerInstanceRandom()
  n18 = VertexColor()
  n19 = Multiply(A=@n17, B=Multiply(A=@n18.R) {constB=0.33})
  n20 = Multiply(A=@n17, B=Multiply(A=@n18.G) {constB=0.7})
  n21 = TextureSample(UVs=Add(A=Multiply(A=@n15, B=@n16), B=AppendVector(A=@n19, B=@n20))) {samplerSource=SSM_Wrap_WorldGroupSettings, samplerType=SAMPLERTYPE_Masks, texture=&T_DirtLeaks_M}
  NamedRerouteDeclaration(None=@n21.G) {name=Details-G_1, nodeColor=<rgba 0.6172 0 1 1>}
  n22 = MaterialFunctionCall!(UseLandscapeCoords, "UseProjection(translucent)") {materialFunction=/BRCoreMaterials/Hermes/Terrain/BiomeMask/MF_Hermes_BiomeMasks}
  n23 = MaterialFunctionCall(Float4=@n22.QuadA_RGBA) {materialFunction=&BreakOutFloat4Components}
  NamedRerouteDeclaration(None=@n23.G) {name=BM_Wetlands, nodeColor=<rgba 0.5 0 1 1>}
  n24 = DotProduct(A=Constant3Vector() {constant=<rgba 0 0 1 0>}, B=VertexNormalWS())
  NamedRerouteDeclaration(None=SphereMask(
    A=ConstantBiasScale(None=@n24)
    B=ScalarParameter() {defaultValue=1, group=Snow, parameterName=G_SnowSample}
    Radius=ScalarParameter() {defaultValue=0.3, group=Snow, parameterName=G_SnowSampleRadius}
    Hardness=ScalarParameter() {defaultValue=0.5, group=Snow, parameterName=G_SnowSampleHardness}
   )) {name=G_SnowTop_Mask, nodeColor=<rgba 0.1953 0 1 1>}
  NamedRerouteDeclaration(None=@n21.B) {name=Details-B_1, nodeColor=<rgba 0.4062 1 0 1>}
  n25 = WorldPosition() {worldPositionShaderOffset=WPT_ExcludeAllShaderOffsets}
  n26 = ScalarParameter() {defaultValue=64, parameterName=WaterLevelOffset}
  n27 = Add(
   A=Add(A=ComponentMask(None=@n25.XYZ) {b=true}, B=NamedRerouteDeclaration(None=@n26) {name=WaterLevelOffset, nodeColor=<rgba 1 0 1 1>})
   B=Constant() {desc=slightly wet sand height offset, r=10}
  )
  n28 = ScalarParameter() {defaultValue=90, parameterName=WaterLevelTransition}
  NamedRerouteDeclaration(None=Saturate(None=Divide(A=@n27, B=@n28))) {name=OceanMask, nodeColor=<rgba 0.5 0 1 1>}
  n29 = TextureSample(UVs=@n12) {
   samplerSource=SSM_Wrap_WorldGroupSettings
   samplerType=SAMPLERTYPE_Normal
   texture=/Game/Environments/Apollo/Rocks/Natural_Rocks/Textures/T_RockDetails_02_N
  }
  NamedRerouteDeclaration(None=MaterialFunctionCall!(Normals=@n29.RGB, TransformToTangentSpace, Normal Strength) {materialFunction=/Game/Athena/Apollo/Environments/Landscape/Materials/Biome/CommonMaterials/MF_HQ_WorldAlignedNormal}.Result) {name=Details N, nodeColor=<rgba 0 1 0.5078 1>}
  NamedRerouteDeclaration(None=@n23.B) {name=BM_MossyForest, nodeColor=<rgba 0.5 0 1 1>}
  n30 = NamedRerouteUsage()
  n31 = RuntimeVirtualTextureSampleParameter() {materialType=BaseColor_Normal_Specular_YCoCg, parameterName=RVT, virtualTexture=/Game/Athena/Apollo/Maps/Landscape/VT/RVT_Apollo_Landscape}
  n32 = SetMaterialAttributes!(
   MaterialAttributes
   Base Color=@n31.BaseColor
   Specular=@n31.BaseColor
   Roughness=@n31.BaseColor
   Emissive Color=Constant()
   Normal=@n31.BaseColor
  ) {attributeSetTypes [
    #1
    #2
    #3
    #6
    #4
   ]}
  n33 = NamedRerouteUsage()
  n34 = GetMaterialAttributes(None=@n33) {attributeGetTypes [#4]}
  n35 = WorldPosition()
  n36 = RuntimeVirtualTextureSampleParameter() {materialType=WorldHeight, parameterName=RVTHeight, virtualTexture=/Game/Athena/Apollo/Maps/Landscape/VT/RVT_Apollo_Landscape_Height}
  n37 = Subtract(A=Add(A=ScalarParameter() {group=RVT, parameterName=Offset RVT WP}, B=ComponentMask(None=@n35.XYZ) {b=true}), B=@n36.WorldHeight)
  n38 = ScalarParameter() {defaultValue=1, group=RVT, parameterName=Blend RVT}
  n39 = LinearInterpolate(
   A=Multiply(A=ScalarParameter() {defaultValue=1, group=RVT, parameterName=RVT Scale Distance Blend}, B=@n38) {constA=3}
   B=@n38
   Alpha=PerInstanceFadeAmount()
  )
  n40 = BlendMaterialAttributes(
   A=@n32
   B=StaticSwitchParameter(
    True=SetMaterialAttributes!(MaterialAttributes=@n32, Normal=@n34.Normal) {attributeSetTypes [#4]}
    False=@n33
   ) {parameterName=RVT Colors}
   Alpha=Saturate(None=Divide(A=@n37, B=@n39))
  ) {vertexAttributeBlendType=UseA}
  NamedRerouteDeclaration(None=MaterialFunctionCall(NoRVT=@n30, RVT=MaterialProxyReplace(Realtime=@n40, MaterialProxy=@n30)) {materialFunction=&FN_RVT_Switch_Material}.Result) {name=WithRVT, nodeColor=<rgba 1 0 0.2109 1>}
  n41 = MaterialFunctionCall(Float4=@n22.QuadC_RGBA) {materialFunction=&BreakOutFloat4Components}
  NamedRerouteDeclaration(None=@n41.G) {name=BM_MaskMaker, nodeColor=<rgba 0.5 0 1 1>}
  n42 = VectorParameter() {group=Gravel, parameterName=Gravel_BaseColor, sortPriority=-500, defaultValue=<rgba 0.15 0.1132 0.1068 0>}
  n43 = ComponentMask(None=@n42.RGB) {b=true, g=true, r=true}
  n44 = VectorParameter() {group=Gravel, parameterName=Gravel_BaseColor_VertexRed, sortPriority=-400, defaultValue=<rgba 0.15 0.084 0.03 0>}
  n45 = ComponentMask(None=@n44.RGB) {b=true, g=true, r=true}
  n46 = MaterialFunctionCall(Float3=MaterialFunctionCall(Hue Shift Percentage=PerInstanceRandom(), Texture=VertexColor()) {materialFunction=&HueShift}.Result) {materialFunction=&BreakOutFloat3Components}
  n47 = VectorParameter() {group=Gravel, parameterName=Gravel_BaseColor_VertexGreen, sortPriority=-300, defaultValue=<rgba 0.13 0.0926 0.052 1>}
  n48 = VectorParameter() {group=Gravel, parameterName=Gravel_BaseColor_VertexBlue, sortPriority=-200, defaultValue=<rgba 0.13 0.0987 0.065 1>}
  n49 = LinearInterpolate(
   A=LinearInterpolate(A=LinearInterpolate(A=@n43, B=@n45, Alpha=@n46.R), B=ComponentMask(None=@n47.RGB) {b=true, g=true, r=true}, Alpha=@n46.G)
   B=ComponentMask(None=@n48.RGB) {b=true, g=true, r=true}
   Alpha=@n46.B
  )
  n50 = Saturate(None=Multiply(A=ScalarParameter() {defaultValue=1.25, group=Gravel, parameterName=Gravel_BaseMacroHueMod, sortPriority=-125}, B=@n49))
  n51 = TextureSample(UVs=MaterialFunctionCall(Scale=ScalarParameter() {defaultValue=200, group=Gravel, parameterName=Gravel_VariationTextureScale}) {materialFunction=&MF_SimpleWorldUV}.Result) {
   samplerSource=SSM_Wrap_WorldGroupSettings
   samplerType=SAMPLERTYPE_Grayscale
   texture=/Game/Weapons/FORT_Melee/Pickaxe_Cube_Ninja_Male/FX/T_CellNoise
  }
  n52 = Saturate(None=Multiply(A=@n51.R, B=ScalarParameter() {defaultValue=3, group=Gravel, parameterName=Gravel_VariationTextureMod}))
  n53 = MaterialFunctionCall() {materialFunction=&MF_Hermes_BiomeMasks}
  n54 = MaterialFunctionCall!(
   Material=SetMaterialAttributes!(MaterialAttributes, Base Color=LinearInterpolate(A=@n50, B=@n49, Alpha=@n52)) {attributeSetTypes [#1]}
   TerrainMasks_QuadA=@n53.QuadA_RGBA
   TerrainMasks_QuadB=@n53.QuadA_RGBA
   TerrainMasks_QuadC=@n53.QuadA_RGBA
   TerrainMasks_QuadD=@n53.QuadA_RGBA
   GrassNoise
  ) {materialFunction=/BRCoreMaterials/Hermes/Terrain/Landscape/Materials/Layers/BiomeColor/MF_Hermes_Gravel_BiomeColor}
  n55 = ScalarParameter() {group=Gravel, parameterName=Gravel_SpecMin}
  n56 = ScalarParameter() {defaultValue=0.7, group=Gravel, parameterName=Gravel_SpecMax}
  n57 = NamedRerouteUsage()
  n58 = ScalarParameter() {defaultValue=0.2, group=Gravel, parameterName=Gravel_RoughnessMin}
  n59 = ScalarParameter() {defaultValue=0.9, group=Gravel, parameterName=Gravel_RoughnessMax}
  n60 = NamedRerouteUsage()
  n61 = NamedRerouteUsage()
  n62 = SetMaterialAttributes!(
   MaterialAttributes
   Base Color=MaterialFunctionCall(
    Base=GetMaterialAttributes(None=@n54.Result) {attributeGetTypes [#1]}.Base Color
    Blend=NamedRerouteUsage()
   ) {materialFunction=&Blend_Overlay}.Result
   Specular=Saturate(None=LinearInterpolate(A=@n55, B=@n56, Alpha=@n57))
   Roughness=Saturate(None=LinearInterpolate(A=@n58, B=@n59, Alpha=@n60))
   Normal=LinearInterpolate(
    A=Transform(None=@n61)
    B=NamedRerouteUsage()
    Alpha=ScalarParameter() {defaultValue=0.4, group=Gravel, parameterName=Gravel_NormalTypeBlend}
   )
  ) {attributeSetTypes [
    #1
    #2
    #3
    #4
   ]}
  NamedRerouteDeclaration(None=@n62) {name=Gravel_Material, nodeColor=<rgba 1 0 0.2109 1>}
  NamedRerouteDeclaration(None=@n23.A) {name=BM_BorealFall, nodeColor=<rgba 0.5 0 1 1>}
  n63 = MaterialFunctionCall(Float4=@n22.QuadB_RGBA) {materialFunction=&BreakOutFloat4Components}
  NamedRerouteDeclaration(None=@n63.R) {name=BM_MagicalForest, nodeColor=<rgba 0.5 0 1 1>}
  NamedRerouteDeclaration(None=@n63.G) {name=BM_RedForest, nodeColor=<rgba 0.5 0 1 1>}
  n64 = Constant3Vector() {constant=<rgba 0 0 1 0>}
  NamedRerouteDeclaration(None=VertexInterpolator(VS=Transform(None=@n64) {transformSourceType=TRANSFORMSOURCE_Instance}).PS) {name=Instance_Z_To_World, nodeColor=<rgba 0 0 1 1>}
  NamedRerouteDeclaration(None=@n41.B) {name=BM_Bigsplat, nodeColor=<rgba 0.5 0 1 1>}
  NamedRerouteDeclaration(None=@n21.R) {name=Details-R_1, nodeColor=<rgba 0 1 0.5078 1>}
  NamedRerouteDeclaration(None=@n14.R) {name=Details-R, nodeColor=<rgba 0 1 0.5078 1>}
  n65 = MaterialFunctionCall() {materialFunction=&MF_Hermes_BiomeMasks}
  n66 = MaterialFunctionCall!(
   Material=SetMaterialAttributes!(MaterialAttributes, Base Color=NamedRerouteUsage()) {attributeSetTypes [#1]}
   TerrainMasks_QuadA=@n65.QuadA_RGBA
   TerrainMasks_QuadB=@n65.QuadA_RGBA
   TerrainMasks_QuadC=@n65.QuadA_RGBA
   TerrainMasks_QuadD=@n65.QuadA_RGBA
   GrassNoise
  ) {materialFunction=/BRCoreMaterials/Hermes/Terrain/Landscape/Materials/Layers/BiomeColor/MF_Hermes_Scree_BiomeColor}
  n67 = NamedRerouteUsage()
  n68 = SetMaterialAttributes!(
   MaterialAttributes
   Base Color=MaterialFunctionCall(
    Base=GetMaterialAttributes(None=@n66.Result) {attributeGetTypes [#1]}.Base Color
    Blend=NamedRerouteUsage()
   ) {materialFunction=&Blend_Overlay}.Result
   Specular=LinearInterpolate(
    A=ScalarParameter() {defaultValue=0.25, group=Scree, parameterName=Scree_Spec_Min}
    B=ScalarParameter() {defaultValue=1, group=Scree, parameterName=Scree_Spec_Max}
    Alpha=NamedRerouteUsage()
   )
   Roughness=LinearInterpolate(
    A=ScalarParameter() {defaultValue=0.1, group=Scree, parameterName=Scree_Rough_Min}
    B=ScalarParameter() {defaultValue=1, group=Scree, parameterName=Scree_Rough_Max}
    Alpha=NamedRerouteUsage()
   )
   Normal=LinearInterpolate(A=Transform(None=@n67), B=NamedRerouteUsage(), Alpha=ScalarParameter() {defaultValue=0.5, group=Gravel, parameterName=Scree_NormalTypeBlend})
  ) {attributeSetTypes [
    #1
    #2
    #3
    #4
   ]}
  NamedRerouteDeclaration(None=@n68) {name=Scree_Material, nodeColor=<rgba 1 0 0.2109 1>}
  n69 = VectorParameter() {group=ForestFloorLeaf_Colors, parameterName=LeafColor_Grasslands_01_R, defaultValue=<rgba 0.095 0.1615 0.0286 1>}
  n70 = VectorParameter() {group=ForestFloorLeaf_Colors, parameterName=LeafColor_Grasslands_02_G, defaultValue=<rgba 0.1215 0.1875 0.0143 1>}
  n71 = VertexColor()
  n72 = VectorParameter() {group=ForestFloorLeaf_Colors, parameterName=LeafColor_Grasslands_03_B, defaultValue=<rgba 0.1374 0.2232 0.0093 1>}
  n73 = LinearInterpolate(
   A=LinearInterpolate(A=ComponentMask(None=@n69.RGB) {b=true, g=true, r=true}, B=ComponentMask(None=@n70.RGB) {b=true, g=true, r=true}, Alpha=@n71.G)
   B=ComponentMask(None=@n72.RGB) {b=true, g=true, r=true}
   Alpha=@n71.B
  )
  n74 = VectorParameter() {group=ForestFloorLeaf_Colors, parameterName=LeafColor_Grasslands_04_A, defaultValue=<rgba 0.1128 0.1927 0.0251 1>}
  n75 = ComponentMask(None=@n74.RGB) {b=true, g=true, r=true}
  n76 = VectorParameter() {group=ForestFloorLeaf_Colors, parameterName=LeafColor_Mountains_01_R, defaultValue=<rgba 0.016 0.1095 0.0176 1>}
  n77 = VectorParameter() {group=ForestFloorLeaf_Colors, parameterName=LeafColor_Mountains_02_G, defaultValue=<rgba 0.0252 0.1683 0.0369 1>}
  n78 = VertexColor()
  n79 = VectorParameter() {group=ForestFloorLeaf_Colors, parameterName=LeafColor_Mountains_03_B, defaultValue=<rgba 0.1927 0.1713 0.0529 1>}
  n80 = LinearInterpolate(
   A=LinearInterpolate(A=ComponentMask(None=@n76.RGB) {b=true, g=true, r=true}, B=ComponentMask(None=@n77.RGB) {b=true, g=true, r=true}, Alpha=@n78.G)
   B=ComponentMask(None=@n79.RGB) {b=true, g=true, r=true}
   Alpha=@n78.B
  )
  n81 = VectorParameter() {group=ForestFloorLeaf_Colors, parameterName=LeafColor_Mountains_04_A, defaultValue=<rgba 0.0833 0.1441 0.0225 1>}
  n82 = ComponentMask(None=@n81.RGB) {b=true, g=true, r=true}
  n83 = VectorParameter() {group=ForestFloorLeaf_Colors, parameterName=LeafColor_BorealFall_01_R, defaultValue=<rgba 0.5457 0.0723 0.0497 1>}
  n84 = VectorParameter() {group=ForestFloorLeaf_Colors, parameterName=LeafColor_BorealFall_02_G, defaultValue=<rgba 0.6939 0.162 0.0222 1>}
  n85 = VertexColor()
  n86 = LinearInterpolate(A=ComponentMask(None=@n83.RGB) {b=true, g=true, r=true}, B=ComponentMask(None=@n84.RGB) {b=true, g=true, r=true}, Alpha=@n85.G)
  n87 = VectorParameter() {group=ForestFloorLeaf_Colors, parameterName=LeafColor_BorealFall_03_B, defaultValue=<rgba 0.2232 0.2623 0.0056 1>}
  n88 = ComponentMask(None=@n87.RGB) {b=true, g=true, r=true}
  n89 = VectorParameter() {group=ForestFloorLeaf_Colors, parameterName=LeafColor_BorealFall_04_A, defaultValue=<rgba 0.5677 0.3068 0.0266 1>}
  n90 = LinearInterpolate(
   A=LinearInterpolate(A=LinearInterpolate(A=@n73, B=@n75, Alpha=@n71.A), B=LinearInterpolate(A=@n80, B=@n82, Alpha=@n78.A), Alpha=NamedRerouteUsage())
   B=LinearInterpolate(A=LinearInterpolate(A=@n86, B=@n88, Alpha=@n85.B), B=ComponentMask(None=@n89.RGB) {b=true, g=true, r=true}, Alpha=@n85.A)
   Alpha=NamedRerouteUsage()
  )
  n91 = VectorParameter() {group=ForestFloorLeaf_Colors, parameterName=LeafColor_MagicalForest_01_R, defaultValue=<rgba 0.1354 0.1152 0.0284 1>}
  n92 = ComponentMask(None=@n91.RGB) {b=true, g=true, r=true}
  n93 = VectorParameter() {group=ForestFloorLeaf_Colors, parameterName=LeafColor_MagicalForest_02_G, defaultValue=<rgba 0.0932 0.1406 0.0242 1>}
  n94 = ComponentMask(None=@n93.RGB) {b=true, g=true, r=true}
  n95 = VertexColor()
  n96 = VectorParameter() {group=ForestFloorLeaf_Colors, parameterName=LeafColor_MagicalForest_03_B, defaultValue=<rgba 0.1242 0.151 0.033 1>}
  n97 = VectorParameter() {group=ForestFloorLeaf_Colors, parameterName=LeafColor_MagicalForest_04_A, defaultValue=<rgba 0.3229 0.2658 0.0336 1>}
  n98 = LinearInterpolate(
   A=LinearInterpolate(A=LinearInterpolate(A=@n92, B=@n94, Alpha=@n95.G), B=ComponentMask(None=@n96.RGB) {b=true, g=true, r=true}, Alpha=@n95.B)
   B=ComponentMask(None=@n97.RGB) {b=true, g=true, r=true}
   Alpha=@n95.A
  )
  n99 = NamedRerouteUsage()
  n100 = VectorParameter() {group=ForestFloorLeaf_Colors, parameterName=LeafColor_MaskMaker_01_R, defaultValue=<rgba 0.1354 0.1245 0.0777 1>}
  n101 = VectorParameter() {group=ForestFloorLeaf_Colors, parameterName=LeafColor_MaskMaker_02_G, defaultValue=<rgba 0.1167 0.1406 0.0819 1>}
  n102 = VertexColor()
  n103 = VectorParameter() {group=ForestFloorLeaf_Colors, parameterName=LeafColor_MaskMaker_03_B, defaultValue=<rgba 0.137 0.151 0.0893 1>}
  n104 = LinearInterpolate(
   A=LinearInterpolate(A=ComponentMask(None=@n100.RGB) {b=true, g=true, r=true}, B=ComponentMask(None=@n101.RGB) {b=true, g=true, r=true}, Alpha=@n102.G)
   B=ComponentMask(None=@n103.RGB) {b=true, g=true, r=true}
   Alpha=@n102.B
  )
  n105 = VectorParameter() {group=ForestFloorLeaf_Colors, parameterName=LeafColor_MaskMaker_04_A, defaultValue=<rgba 0.125 0.1015 0.0793 1>}
  n106 = ComponentMask(None=@n105.RGB) {b=true, g=true, r=true}
  n107 = VectorParameter() {group=ForestFloorLeaf_Colors, parameterName=LeafColor_RedForest_01_R, defaultValue=<rgba 0.3938 0.0139 0.0422 1>}
  n108 = VectorParameter() {group=ForestFloorLeaf_Colors, parameterName=LeafColor_RedForest_02_G, defaultValue=<rgba 0.4397 0.008 0.0378 1>}
  n109 = VertexColor()
  n110 = LinearInterpolate(A=ComponentMask(None=@n107.RGB) {b=true, g=true, r=true}, B=ComponentMask(None=@n108.RGB) {b=true, g=true, r=true}, Alpha=@n109.G)
  n111 = VectorParameter() {group=ForestFloorLeaf_Colors, parameterName=LeafColor_RedForest_03_B, defaultValue=<rgba 0.6354 0.045 0.0475 1>}
  n112 = ComponentMask(None=@n111.RGB) {b=true, g=true, r=true}
  n113 = VectorParameter() {group=ForestFloorLeaf_Colors, parameterName=LeafColor_RedForest_04_A, defaultValue=<rgba 0.4479 0.0187 0.0562 1>}
  n114 = LinearInterpolate(
   A=LinearInterpolate(A=LinearInterpolate(A=@n90, B=@n98, Alpha=@n99), B=LinearInterpolate(A=@n104, B=@n106, Alpha=@n102.A), Alpha=NamedRerouteUsage())
   B=LinearInterpolate(A=LinearInterpolate(A=@n110, B=@n112, Alpha=@n109.B), B=ComponentMask(None=@n113.RGB) {b=true, g=true, r=true}, Alpha=@n109.A)
   Alpha=NamedRerouteUsage()
  )
  n115 = VectorParameter() {group=ForestFloorLeaf_Colors, parameterName=LeafColor_Bigsplat_01_R, defaultValue=<rgba 0.1064 0.125 0.0451 1>}
  n116 = ComponentMask(None=@n115.RGB) {b=true, g=true, r=true}
  n117 = VectorParameter() {group=ForestFloorLeaf_Colors, parameterName=LeafColor_Bigsplat_02_G, defaultValue=<rgba 0.05 0.25 0.0943 1>}
  n118 = ComponentMask(None=@n117.RGB) {b=true, g=true, r=true}
  n119 = VertexColor()
  n120 = VectorParameter() {group=ForestFloorLeaf_Colors, parameterName=LeafColor_Bigsplat_03_B, defaultValue=<rgba 0.04 0.2 0.1308 1>}
  n121 = VectorParameter() {group=ForestFloorLeaf_Colors, parameterName=LeafColor_Bigsplat_04_A, defaultValue=<rgba 0.1741 0.18 0.0492 1>}
  n122 = LinearInterpolate(
   A=LinearInterpolate(A=LinearInterpolate(A=@n116, B=@n118, Alpha=@n119.G), B=ComponentMask(None=@n120.RGB) {b=true, g=true, r=true}, Alpha=@n119.B)
   B=ComponentMask(None=@n121.RGB) {b=true, g=true, r=true}
   Alpha=@n119.A
  )
  n123 = NamedRerouteUsage()
  n124 = ScalarParameter() {defaultValue=0.3, group=ForestFloorLeaf_Settings, parameterName=FFLeaves_SpecMin}
  n125 = ScalarParameter() {defaultValue=0.7, group=ForestFloorLeaf_Settings, parameterName=FFLeaves_FFLeaves_SpecMax}
  n126 = ComponentMask(None=NamedRerouteUsage()) {r=true}
  n127 = ScalarParameter() {defaultValue=1, group=ForestFloorLeaf_Settings, parameterName=FFLeaves_RoughnessMin}
  n128 = ScalarParameter() {defaultValue=0.5, group=ForestFloorLeaf_Settings, parameterName=FFLeaves_RoughnessMax}
  n129 = SetMaterialAttributes!(
   MaterialAttributes
   Base Color=Multiply(A=NamedRerouteUsage(), B=LinearInterpolate(A=@n114, B=@n122, Alpha=@n123))
   Specular=Saturate(None=LinearInterpolate(A=@n124, B=@n125, Alpha=@n126))
   Roughness=Saturate(None=LinearInterpolate(A=@n127, B=@n128, Alpha=@n126))
   Normal=Transform(None=NamedRerouteUsage())
   Opacity Mask=NamedRerouteUsage()
  ) {attributeSetTypes [
    #1
    #2
    #3
    #4
    #5
   ]}
  NamedRerouteDeclaration(None=@n129) {name=ForestFloorLeaves_Material, nodeColor=<rgba 1 0 0.2109 1>}
  NamedRerouteDeclaration(None=@n23.R) {name=BM_Mountains, nodeColor=<rgba 0.5 0 1 1>}
  NamedRerouteDeclaration(None=@n14.G) {name=Details-G, nodeColor=<rgba 0.6172 0 1 1>}
  Reroute(None=@n7)
  n130 = GetMaterialAttributes(None=@n7) {attributeGetTypes [#1]}
  n131 = MaterialFunctionCall!(
   ScalarWeightMask
   UseSnowOcclusionTextureMask
   ------
   UseMaterialInput
   AdditionalMaskToRemoveSnow
   OutputTangentNormals
   UseAdditionalMaskToRemoveSnow
   UseSnowNormals
   OuputVertexNormalsAsTheSnowNormals
   UseSparkle
   LQSpark
   AllowSnowMicroNormals
   CliffMaterialOnly
   Material
   CliffLandscapeLayerMask
   IsLandscape
   AreInputNormalsTangentSpace
   UseInputNormalsForFakeSnowShadow
  ) {materialFunction=/Game/Athena/Helios/Environments/Landscape/Materials/QualityLevelLayers/MF_Helios_Snow_Default}
  n132 = NamedRerouteUsage()
  SetMaterialAttributes!(
   MaterialAttributes=@n130.MaterialAttributes
   Base Color=LinearInterpolate(
    A=@n130.Base Color
    B=VectorParameter() {group=Snow, parameterName=G_SnowColor, defaultValue=<rgba 0.4292 0.4575 0.5 1>}.RGB
    Alpha=Multiply(A=@n131.DetailMask, B=@n132)
   )
  ) {attributeSetTypes [#1]}
  n133 = NamedRerouteUsage()
  n134 = StaticSwitchParameter(True=NamedRerouteUsage(), False=@n133) {parameterName=UseRVT}
  ShadingPathSwitch(Default=@n134, Deferred=@n134, Forward=@n133, Mobile=@n133)
 ]
}
